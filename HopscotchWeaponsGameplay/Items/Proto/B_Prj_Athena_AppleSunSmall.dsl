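; Small AppleSun projectile Blueprint, derived from a parent projectile, whose apple mesh is driven by a rotating movement component.
; Capsule hits clamp the projectile's speed and play an attached bounce sound when the hit is not valid for spawning.
; Hit events run parent logic except when the hit actor casts to an AppleSun projectile, so those collisions are ignored.
; When it stops, it runs parent logic and deactivates the Niagara trail.
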(event OnComponentHit(CapsuleComponent)
  (params (HitComponent :type Primitive%20Component%20Object%20Reference) (OtherActor :type Actor%20Object%20Reference) (OtherComp :type Primitive%20Component%20Object%20Reference) (NormalImpulse :type Vector :default "0, 0, 0") (Hit :type Hit%20Result%20Structure%20%28by%20ref%29))
  (Projectile|SetVelocity (Variables|Movement|GetProjectileMovementComponent) (Math|Vector|ClampVectorSize (Variables|/Script/Engine.MovementComponent|GetVelocity (Variables|Movement|GetProjectileMovementComponent)) 0.0 1.2e+03))
  ;[0.1588 0.1365 0.3854] Audio
  (bind _validspawn (CallFunction|IsHitValidforSpawn HitComponent OtherActor Hit))
  (if (not _validspawn)
    (Audio|SpawnSoundAttached (Variables|Default|GetProjectile%20Bounce%20Sound) (Variables|Collision|GetCapsuleComponent))))

(event EventOnStop
  (params (Hit :type Hit%20Result%20Structure%20%28by%20ref%29))
  (|CallParentFunction Hit)
  (Components|Activation|SetActive (Variables|Default|GetNS_AppleSun_Projectile_Trail_Converted) false false))

(event Collision|EventHit
  (params (MyComp :type Primitive%20Component%20Object%20Reference) (Other :type Actor%20Object%20Reference) (OtherComp :type Primitive%20Component%20Object%20Reference) (bSelfMoved :type Boolean) (HitLocation :type Vector :default "0, 0, 0") (HitNormal :type Vector :default "0, 0, 0") (NormalImpulse :type Vector :default "0, 0, 0") (Hit :type Hit%20Result%20Structure%20%28by%20ref%29))
  (bind (_bblockinghit _binitialoverlap _time _distance _location _impactpoint _normal _impactnormal _physmat _hitactor) (Collision|BreakHitResult Hit))
  ;[1 1 1] Ignore collision with other AppleSun projectiles
  (Utilities|Casting|CastToB_Prj_Athena_AppleSun_ _hitactor
    (:CastFailed
      (|CallParentFunction MyComp Other OtherComp bSelfMoved HitLocation HitNormal NormalImpulse Hit))))

(fn ConstructionScript
  (|CallParentFunction)
  (Components|Movement|SetUpdatedComponent (Variables|Default|GetRotatingMovement) (Variables|Default|GetApple_Sun)))
